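# Looping ambient sound cue for turbines in Athena, routed to the outdoor ambient sound class.
# A mixer layers a border wind loop at reduced level beneath a full-level metallic turbine loop.
# Overridden attenuation uses a non-spatialized box volume with falloff, so it acts as an area bed rather than a point source.
# Occlusion traces on a game channel apply an interpolated low-pass filter when blocked.
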SoundCue /Figment_S09/Athena/Sounds/Ambient/Turbines/Turbine_Ambient_Loop
root=Mixer_0
nodes [
 {id=Mixer_0, children [WavePlayer_0,WavePlayer_1], properties {inputVolume [0.35,1]}}
 {id=WavePlayer_0, properties {bLooping=true, soundWaveAssetPtr=&/border_wind_05Loop}}
 {id=WavePlayer_1, properties {bLooping=true, soundWaveAssetPtr=&/Turbine_Ambient_Metal_Loop}}
]
properties {
 bOverrideAttenuation=true
 maxDistance=1000
 soundClassObject=&Fort_Ambient_Outdoors_Athena
 volumeMultiplier=0.7
 attenuationOverrides {
  attenuationShape=Box
  bEnableOcclusion=true
  bSpatialize=false
  falloffDistance=500
  occlusionInterpolationTime=1
  occlusionLowPassFilterFrequency=2000
  occlusionTraceChannel=ECC_GameTraceChannel11
  attenuationShapeExtents=<500 500 500>
 }
}
